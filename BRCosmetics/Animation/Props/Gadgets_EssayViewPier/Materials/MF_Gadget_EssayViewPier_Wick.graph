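# Material function for the EssayViewPier gadget prop's wick, adding burn coloring and cutout to incoming material attributes.
# Emissive blends toward dark and light burn colors, and opacity mask uses a remapped UV-V threshold plus WickExists, both confined by green-channel masks.
# A vertex-interpolated timestamp transition provides the wick timing, with a static switch to looping debug time.
# Some nodes, such as WickOff and a modulo-time call, are left unconnected.
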MaterialFunction /BRCosmetics/Animation/Props/Gadgets_EssayViewPier/Materials/MF_Gadget_EssayViewPier_Wick
graph {
 properties {previewBlendMode=BLEND_Masked}
 nodes [
  ComponentMask(None=TextureCoordinate()) {g=true}
  n1 = GetMaterialAttributes(None=NamedRerouteUsage()) {attributeGetTypes [#1, #2]}
  n2 = ComponentMask(None=NamedRerouteUsage()) {g=true}
  n3 = NamedRerouteUsage()
  n4 = MaterialFunctionCall(
   A=@n1.EmissiveColor
   B=VectorParameter() {group=EssayViewPier Tech Art, parameterName=DarkBurnWick, defaultValue=<rgba 0.026 0.0169 0.0159 1>}.RGB
   C=VectorParameter() {group=EssayViewPier Tech Art, parameterName=LightBurnWick, defaultValue=<rgba 2 0.2781 0.2188 1>}.RGB
   Alpha=SmoothStep(Value=Multiply(A=@n2, B=@n3)) {constMin=0.25}
  ) {materialFunction=&Lerp_3Color}
  n5 = ComponentMask(None=NamedRerouteUsage()) {g=true}
  FunctionOutput(None=SetMaterialAttributes!(
    MaterialAttributes=@n1.MaterialAttributes
    EmissiveColor=LinearInterpolate(A=@n1.EmissiveColor, B=@n4.Result, Alpha=@n5)
    OpacityMask=NamedRerouteUsage()
   ) {attributeSetTypes [#1, #2]})
  MaterialFunctionCall!(
   Time
   Divisor=ScalarParameter() {defaultValue=1.5, group=EssayViewPier Tech Art, parameterName=DebugTimeDivisor, sortPriority=30}
   Mod=ScalarParameter() {defaultValue=2, group=EssayViewPier Tech Art, parameterName=DebugTimeModulo, sortPriority=30}
  ) {materialFunction=/Game/Effects/Fort_Effects/Materials/_Functions/MF_ModuloTime}
  n6 = ScalarParameter() {defaultValue=1, group=EssayViewPier Tech Art, parameterName=WickExists}
  Multiply(B=@n6)
  NamedRerouteDeclaration(None=FunctionInput() {inputName=In_MaterialAttribute, inputType=FunctionInput_MaterialAttributes}) {name=In_MaterialAttribute, nodeColor=<rgba 1 0.3516 0 1>}
  n7 = TextureSample() {samplerType=SAMPLERTYPE_LinearColor, texture=/BRCosmetics/Animation/Props/Gadgets_EssayViewPier/Textures/T_EssayViewPier_Emote_Prop_FX2}
  NamedRerouteDeclaration(None=FunctionInput(Preview=@n7.RGB) {bUsePreviewValueAsDefault=true, inputName=In_TechArtMask, sortPriority=1}) {name=In_TechArtMask, nodeColor=<rgba 0.0312 0 1 1>}
  n8 = Divide(A=Time(), B=ScalarParameter() {defaultValue=3, group=EssayViewPier Tech Art, parameterName=DebugTransitionLength, sortPriority=30})
  n9 = MaterialFunctionCall(
   TimestampA=ScalarParameter() {defaultValue=1, group=EssayViewPier Tech Art, parameterName=Timestamp, sortPriority=30}
   TransitionLength_In=ScalarParameter() {defaultValue=0.1, group=EssayViewPier Tech Art, parameterName=TransitionLength, sortPriority=30}
  ) {materialFunction=&MF_Timestamp_InOut_01}
  NamedRerouteDeclaration(None=StaticSwitchParameter(True=Frac(None=@n8), False=VertexInterpolator(VS=@n9.Linear).PS) {group=FNCSShield_Screen_Graphics, parameterName=DebugModeOn, sortPriority=0}) {name=TimeStampWick, nodeColor=<rgba 0.125 1 0 1>}
  n10 = TextureCoordinate() {uTiling=12, vTiling=12}
  n11 = TextureObjectParameter() {group=EssayViewPier Tech Art, parameterName=NoiseForOpacityErosionOnWick}
  n12 = NamedRerouteUsage()
  n13 = Subtract(A=OneMinus(None=TextureSample(UVs=@n10, Tex=@n11).R), B=@n12)
  n14 = OneMinus(None=@n12)
  NamedRerouteDeclaration(None=Saturate(None=Divide(A=@n13, B=@n14))) {name=Noise, nodeColor=<rgba 1 0.5859 0 1>}
  n15 = ComponentMask(None=TextureCoordinate()) {g=true}
  n16 = Multiply(A=NamedRerouteUsage()) {constB=0.1}
  n17 = MaterialFunctionCall(
   Input=Saturate(None=Subtract(A=@n15, B=@n16))
   Input Low=ScalarParameter() {defaultValue=0.05, parameterName="Input Low (S)"}
   Input High=ScalarParameter() {defaultValue=0.1, parameterName="Input High (S)"}
  ) {materialFunction=&RemapValue_0-1}
  n18 = NamedRerouteUsage()
  NamedRerouteDeclaration(None=LinearInterpolate(A=@n1.OpacityMask, B=Add(A=@n17.Result, B=@n6), Alpha=ComponentMask(None=@n18) {g=true})) {name=OpacityMask, nodeColor=<rgba 0.2656 0 1 1>}
  n19 = TextureObjectParameter() {
   group=EssayViewPier Tech Art
   parameterName=TechArtMask2
   samplerType=SAMPLERTYPE_LinearColor
   texture=&T_EssayViewPier_Emote_Prop_FX2
  }
  NamedRerouteDeclaration(None=TextureSample(Tex=@n19).RGB) {name=TechArtMask2, nodeColor=<rgba 0 1 0.0156 1>}
  NamedRerouteUsage()
  NamedRerouteUsage()
  NamedRerouteUsage()
  NamedRerouteUsage()
  n20 = NamedRerouteUsage()
  Saturate(None=Subtract(B=Multiply(A=@n20) {constB=1.2}))
  ScalarParameter() {group=EssayViewPier Tech Art, parameterName=WickOff}
 ]
}
